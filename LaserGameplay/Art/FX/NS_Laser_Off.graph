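# Niagara system for a short laser shut-off effect, built from one-shot bursts along an X-axis beam scaled by User.ScaleBeam.
# Glow puts a red sprite and light at each beam end, with light intensity scaled by time of day, visibility distance and flash variables.
# Laser stretches two nested meshes that swell in width while dimming.
# LaserEnergy adds a longer-lived red mesh that widens and fades out.
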NiagaraSystem /LaserGameplay/Art/FX/NS_Laser_Off
dependencies {
 usedDataInterfaces [&NiagaraDataInterfaceArrayFloat, &NiagaraDataInterfaceCurve, &NiagaraDataInterfaceArrayFloat3]
 usedDynamicInputs [
  &FN_LOD_Distance_Remap
  &MakeCustomFloatFromBool
  &Waveform
 ]
 usedModules [
  &SystemState
  &self:SystemSpawnScript.SystemScriptSource.SystemScriptGraph.NiagaraNodeAssignment_0.SetVariables_#1
  &self:SystemSpawnScript.SystemScriptSource.SystemScriptGraph.NiagaraNodeAssignment_1.SetVariables_#2
  &self:SystemSpawnScript.SystemScriptSource.SystemScriptGraph.NiagaraNodeAssignment_2.SetVariables_#3
  &self:SystemSpawnScript.SystemScriptSource.SystemScriptGraph.NiagaraNodeAssignment_3.SetVariables_#4
 ]
}
emitters {
 Glow {
  emitterUpdateScript [
   &EmitterState()
   &SpawnBurst_Instantaneous(Spawn Count=2, Use Spawn Probability=false, UseLoopCountLimit=false)
  ]
  particleSpawnScript [
   &PartitionParticles()
   /SetVariables(Particles.ScaleBeam:NiagaraFloat=&Multiply_Float(A=20000, B=$User.ScaleBeam)) {isSetParameters=true}
   &InitializeParticle(
    Color=<rgba 1.5 0.1377 0.093 10>
    Lifetime=0.25
    Position Offset=&MakeVector(
     X=&Lerp_Float(
      A=0
      B=$Particles.ScaleBeam
      Alpha=&SelectFloatFromArray(Direct Array Index=$Particles.PartitionParticles.Partition, Float Selection Array=DataInterface() {propertyValues {FloatData [0,1]}})
     )
     Y=0
     Z=0
    )
    Sprite Rotation Mode=::NewEnumerator2
    Sprite Size Mode=::NewEnumerator1
    Uniform Sprite Size Max=130
    Uniform Sprite Size Min=100
    UsePositionOffset=true
    Write Material Random=true
   )
   &CameraOffset(Camera Offset Amount=15)
  ]
  particleUpdateScript [
   &ParticleState()
   &ScaleSpriteSize(Uniform Curve Sprite Scale=DataInterface() {propertyValues {Curve {keys [
        {interpMode=RCIM_Cubic, tangentMode=RCTM_User, time=0, value=1}
        {interpMode=RCIM_Cubic, tangentMode=RCTM_User, time=0.3, value=2}
        {interpMode=RCIM_Cubic, tangentMode=RCTM_User, time=1, value=1}
       ]}}})
   &ScaleColor(ScaleA=true, ScaleRGB=false)
   &DynamicMaterialParameters(
    Index 1 Param 1=150
    Module.Param0WriteEnabled1=true
    Module.Param1WriteEnabled1=true
    Module.Param2WriteEnabled1=true
    Module.Param3WriteEnabled1=true
    Write Parameter Index 0=false
    Write Parameter Index 1=true
   )
   &FN_Light_Attributes(
    Color=<rgba 1.5 0.1377 0.093 1>
    Create Color Attribute=true
    Create Exponent Attribute=true
    Create Position Attribute=true
    Create Radius Attribute=true
    Exponent=20
    Position=&Add_Vector(
     A=&ConvertPositionToVector(Input Position=$Particles.Position, Convert To Absolute Large World Vector=false)
     B=&SelectVectorFromArray(
      Direct Array Index=$Particles.PartitionParticles.Partition
      Vector Selection Array=DataInterface() {propertyValues {FloatData [<25 0 0>, <-25 0 0>]}}
     )
    )
    Radius=&Multiply_Float(A=150, B=1)
    Scale Color=&Multiply_Float(
     A=&FNTimeOfDayScalar(Night Scale=1, Morning Scale=1.5, Day Scale=2, Evening Scale=1.5)
     B=&Multiply_Float(A=$System.VisibilityDistance, B=$System.LaserFlash)
    )
   )
   /SetVariables(Particles.Debug:NiagaraFloat=$Particles.ScaleBeam) {enabled=false, isSetParameters=true}
  ]
  rendererClasses [&NiagaraSpriteRendererProperties, &NiagaraLightRendererProperties]
  renderers [
   {rendererClass=&NiagaraSpriteRendererProperties}
   {rendererClass=&NiagaraLightRendererProperties, rendererIndex=1}
  ]
 }
 Laser {
  emitterUpdateScript [
   &EmitterState()
   &SpawnBurst_Instantaneous(Spawn Count=2, Use Spawn Probability=false, UseLoopCountLimit=false)
  ]
  particleSpawnScript [
   &PartitionParticles()
   /SetVariables(
    Particles.ScaleBeam:NiagaraFloat=&Multiply_Float(A=200, B=$User.ScaleBeam)
    Particles.ScaleYZ:NiagaraFloat=&Multiply_Float(
     A=&SelectFloatFromArray(Direct Array Index=$Particles.PartitionParticles.Partition, Float Selection Array=DataInterface() {propertyValues {FloatData [0.03,0.25]}})
     B=1
    )
   ) {isSetParameters=true, module=SetVariables_#5}
   &InitializeParticle(
    Color=&MakeLinearColorFromVectorAndFloat(
     "Vector (RGB)"=&SelectVectorFromArray(
      Direct Array Index=$Particles.PartitionParticles.Partition
      Vector Selection Array=DataInterface() {propertyValues {FloatData [<0.25 0.25 0.25>, <0.1 0.1 0.1>]}}
     )
     "Float (Alpha)"=&SelectFloatFromArray(Direct Array Index=$Particles.PartitionParticles.Partition, Float Selection Array=DataInterface() {propertyValues {FloatData [0.25,0.01]}})
    )
    Lifetime=0.25
    Mesh Scale=&MakeVector(X=$Particles.ScaleBeam, Y=$Particles.ScaleYZ, Z=$Particles.ScaleYZ)
    Mesh Scale Mode=::NewEnumerator4
    UsePositionOffset=false
    Write Material Random=true
   )
  ]
  particleUpdateScript [
   &ParticleState()
   &ScaleMeshSize(Scale Factor=&Lerp_Vector(
     A=<1 0.01 0.01>
     B=<1 10 10>
     Alpha=&FloatFromCurve(FloatCurve=DataInterface() {propertyValues {Curve {keys [
          {interpMode=RCIM_Cubic, leaveTangentWeight=0.981, tangentMode=RCTM_User, time=0, value=0}
          {interpMode=RCIM_Cubic, tangentMode=RCTM_User, time=0.8, value=1}
          {arriveTangent=-10.8953, interpMode=RCIM_Cubic, leaveTangent=-10.8953, tangentMode=RCTM_User, time=1, value=0.2}
         ]}}})
    ))
   &ScaleColor(
    Scale RGB=&Multiply_VectorByFloat(
     Vector=<1 1 1>
     Float=&Lerp_Float(
      A=4
      B=0.1
      Alpha=&FloatFromCurve(FloatCurve=DataInterface() {propertyValues {Curve {keys [{interpMode=RCIM_Cubic, tangentMode=RCTM_User, time=0, value=0}, {interpMode=RCIM_Cubic, tangentMode=RCTM_User, time=1, value=1}]}}})
     )
    )
    ScaleA=true
    ScaleRGB=true
   )
  ]
  rendererClasses [&NiagaraMeshRendererProperties]
  renderers [{rendererClass=&NiagaraMeshRendererProperties, rendererIndex=0}]
 }
 LaserEnergy {
  emitterUpdateScript [
   &EmitterState()
   &SpawnBurst_Instantaneous(Use Spawn Probability=false, UseLoopCountLimit=false)
   /SetVariables(
    Emitter.PanSpeed:NiagaraFloat=&RandomRangeFloat(Minimum=0.2, Maximum=0.4)
    Emitter.WPOAmount:NiagaraFloat=&RandomRangeFloat(Minimum=10, Maximum=20)
   ) {isSetParameters=true}
  ]
  particleSpawnScript [
   &PartitionParticles()
   /SetVariables(Particles.ScaleBeam:NiagaraFloat=&Multiply_Float(A=200, B=$User.ScaleBeam), Particles.ScaleYZ=0.3) {isSetParameters=true, module=SetVariables_#5}
   &InitializeParticle(
    Color=&MakeLinearColorFromVectorAndFloat("Vector (RGB)"=<0.5 0.1 0.1>, "Float (Alpha)"=0.25)
    Lifetime=0.4
    Mesh Scale=&MakeVector(X=$Particles.ScaleBeam, Y=$Particles.ScaleYZ, Z=$Particles.ScaleYZ)
    Mesh Scale Mode=::NewEnumerator4
    UsePositionOffset=false
    Write Material Random=true
   )
  ]
  particleUpdateScript [
   &ParticleState()
   &ScaleMeshSize(Scale Factor=&Lerp_Vector(
     A=<1 0.5 0.5>
     B=<1 5 5>
     Alpha=&FloatFromCurve(FloatCurve=DataInterface() {propertyValues {Curve {keys [
          {arriveTangent=8.8655, interpMode=RCIM_Cubic, leaveTangent=8.8655, tangentMode=RCTM_User, time=0, value=0}
          {interpMode=RCIM_Cubic, tangentMode=RCTM_User, time=0.3, value=1}
          {arriveTangentWeight=1.6162, interpMode=RCIM_Cubic, tangentMode=RCTM_User, time=1, value=1}
         ]}}})
    ))
   &ScaleColor(
    Scale Alpha=&FloatFromCurve(FloatCurve=DataInterface() {propertyValues {Curve {keys [
         {interpMode=RCIM_Cubic, tangentMode=RCTM_User, time=0, value=1}
         {arriveTangentWeight=1.151, interpMode=RCIM_Cubic, tangentMode=RCTM_User, time=1, value=0}
        ]}}})
    ScaleA=true
    ScaleRGB=false
   )
  ]
  rendererClasses [&NiagaraMeshRendererProperties]
  renderers [{rendererClass=&NiagaraMeshRendererProperties, rendererIndex=0}]
 }
}
summary {
 systemName=NS_Laser_Off
 emitters [
  {emitterName=Glow, rendererClasses [&NiagaraSpriteRendererProperties, &NiagaraLightRendererProperties]}
  {emitterName=Laser, rendererClasses [&NiagaraMeshRendererProperties]}
  {emitterName=LaserEnergy, rendererClasses [&NiagaraMeshRendererProperties]}
 ]
 userVariables [
  User.FlashLaser {type=NiagaraBool}
  User.MaxVisibleDistance {type=NiagaraFloat, defaultValue=2000, description=""}
  User.ScaleBeam {type=NiagaraFloat, defaultValue=1, description=""}
  User.UseDistanceVisibility {type=NiagaraBool, defaultValue=true}
 ]
}
