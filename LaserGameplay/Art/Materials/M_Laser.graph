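# M_Laser is an unlit alpha-composite material for instanced laser beams, using SplineThicken world position offset for wobbling thickness.
# A flickering curve-atlas gradient across U sets color and opacity, combined with endcap masking and eye-adaptation compensation.
# Per-instance data drives pulse, time offset, flicker speed and a timestamped sphere-mask transition-in, with direction picked per instance at random.
# Two panned sparkle-noise samples multiply into a dust mask.
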Material /LaserGameplay/Art/Materials/M_Laser
graph {
 properties {
  bEnableResponsiveAA=true
  bUsedWithInstancedStaticMeshes=true
  blendMode=BLEND_AlphaComposite
  pixelDepthOffsetMode=PDOM_Legacy
  shadingModel=MSM_Unlit
  usedShadingModels=Unlit
 }
 nodes [
  n1 = Cosine(None=Add(A=Time(), B=PerInstanceCustomData!(DefaultValue) {dataIndex=3, desc=TimeOffset})) {period=0.5}
  n2 = MaterialFunctionCall(Float2=TextureCoordinate()) {materialFunction=&BreakOutFloat2Components}
  n3 = Abs(None=ConstantBiasScale(None=NamedRerouteDeclaration(None=@n2.R) {name=U Gradient, nodeColor=<rgba 1 0 0.3516 1>}) {bias=-0.5, scale=2})
  n4 = Multiply(
   A=Time()
   B=Multiply(
    A=Constant4Vector() {constant=<rgba 0.0276 0.023 0.0294 0.003>}.RGBA
    B=PerInstanceCustomData() {constDefaultValue=1, dataIndex=1, desc=Random speed flickering}
   )
  )
  n5 = Add(A=Sine(None=@n4) {period=0.0037}, B=Cosine(None=@n4) {period=0.0053})
  n6 = ScalarParameter() {defaultValue=0.007, parameterName=GradientSizeWobble}
  n7 = CurveAtlasRowParameter(CurveTime=Add(A=OneMinus(None=@n3), B=DotProduct(A=@n5, B=@n6))) {atlas=/LaserGameplay/Art/Textures/CA_Laser, curve=&/C_LaserColor, parameterName=ColorAtlas}
  n8 = Multiply(A=LinearInterpolate(B=ConstantBiasScale(None=@n1), Alpha=PerInstanceCustomData() {desc=Pulse}) {constA=1}, B=@n7)
  n9 = Add(A=@n7.A, B=NamedRerouteUsage())
  n10 = Multiply(
   A=NamedRerouteUsage()
   B=SphereMask(A=NamedRerouteUsage(), B=Constant() {r=0.5}, Radius=ScalarParameter() {defaultValue=0.55, parameterName=EndcapRadius}) {hardnessPercent=0}
  )
  EyeAdaptationInverse(LightValueInput=Multiply(A=Multiply(A=@n8, B=@n9), B=@n10))
  n11 = Saturate(None=@n7.R)
  Multiply(A=Multiply(A=OneMinus(None=@n11), B=@n9), B=@n10)
  NamedRerouteDeclaration(None=@n2.G) {name=V Gradient, nodeColor=<rgba 0 0.4609 1 1>}
  n12 = PerInstanceRandom()
  n13 = NamedRerouteUsage()
  n14 = OneMinus(None=@n13)
  n15 = MaterialFunctionCall(
   TimestampA=PerInstanceCustomData() {dataIndex=2, desc=Time Stamp}
   TransitionLength_In=ScalarParameter() {defaultValue=0.22, parameterName=TransitionIn}
   Transition In Only?=StaticBool() {value=true}
  ) {materialFunction=&MF_Timestamp_InOut_01}
  NamedRerouteDeclaration(None=SphereMask(
    A=If(A=@n12, "A > B"=@n13, "A < B"=@n14) {constB=0.5}
    B=Constant()
    Radius=@n15.SmoothStep
    Hardness=LinearInterpolate(Alpha=@n15.SmoothStep) {constA=0.6}
   ) {attenuationRadius=0.05, hardnessPercent=0.6}) {name=LaserAnimation, nodeColor=<rgba 0.4531 1 0 1>}
  n16 = Add(
   A=DotProduct(A=@n5, B=ScalarParameter() {defaultValue=0.3, parameterName=SplineSizeWobble})
   B=ScalarParameter() {defaultValue=20, parameterName=SplineThickness}
  )
  NamedRerouteDeclaration(None=MaterialFunctionCall(WidthBase=@n16, WidthTip=@n16) {materialFunction=&SplineThicken}.WorldPositionOffset) {name=WPO, nodeColor=<rgba 0 1 0.8125 1>}
  n17 = TextureCoordinate() {vTiling=10}
  n18 = VectorParameter() {parameterName=DustA_UVScalePan, defaultValue=<rgba 1 4 0.0399 0.135>}
  n19 = Panner(Coordinate=Multiply(A=@n17, B=AppendVector(A=@n18.R, B=@n18.R)), Speed=AppendVector(A=@n18.B, B=@n18.B))
  n20 = VectorParameter() {parameterName=DustB_UVScalePan, defaultValue=<rgba 1.2 4.5 -0.14 -0.165>}
  n21 = Panner(Coordinate=Multiply(A=@n17, B=AppendVector(A=@n20.R, B=@n20.R)), Speed=AppendVector(A=@n20.B, B=@n20.B))
  NamedRerouteDeclaration(None=Multiply(
    A=TextureSample(UVs=@n19) {samplerType=SAMPLERTYPE_LinearColor, texture=/Game/Gadgets/Assets/Slow_Field/Textures/T_Slowfield_SparkleNoise_01}.R
    B=TextureSample(UVs=@n21) {samplerType=SAMPLERTYPE_LinearColor, texture=&T_Slowfield_SparkleNoise_01}.G
   )) {name=Dust, nodeColor=<rgba 1 0 0.75 1>}
  NamedRerouteUsage()
 ]
}
