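# Lit material for the laser emitter outlet, built for instanced static meshes and Nanite, with an emissive glow that shows enabled or disabled state.
# A texture mask shaped by Fresnel feeds two colors: the active one scaled by per-instance intensity and a timestamp transition, the disabled one pulsing on an eased sine.
# Per-instance Disabled data blends between them, and EyeAdaptationInverse compensates the result for auto exposure.
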Material /LaserGameplay/Art/Materials/M_LaserOutlet
graph {
 properties {bUsedWithInstancedStaticMeshes=true, bUsedWithNanite=true, pixelDepthOffsetMode=PDOM_Legacy, usedShadingModels=Default Lit}
 nodes [
  n1 = MaterialFunctionCall(
   TimestampA=PerInstanceCustomData!(DefaultValue) {dataIndex=2, desc=Timestamp}
   TransitionLength_In=PerInstanceCustomData() {dataIndex=3, desc=TimestampTransition}
  ) {materialFunction=&MF_Timestamp_InOut_01}
  n2 = Constant() {r=2}
  n3 = Multiply(
   A=TextureSample() {samplerType=SAMPLERTYPE_Masks, texture=/LaserGameplay/Art/Textures/T_Gameplay_LaserEmitter_A_A_M}.G
   B=LinearInterpolate(Alpha=MaterialFunctionCall(Exponent=@n2) {materialFunction=&MF_Fresnel_FX}.Fresnel) {constA=0.5}
  )
  n4 = Multiply(A=Multiply(A=@n3, B=VectorParameter() {parameterName=EmissiveColor, defaultValue=<rgba 25 0.0208 0 0>}.RGB), B=PerInstanceCustomData())
  n5 = Time()
  n6 = ConstantBiasScale(None=Sine(None=@n5) {period=2})
  n7 = Constant() {r=0.0427}
  n8 = Constant() {r=2.3965}
  n9 = Constant() {r=0.8553}
  n10 = Multiply(
   A=VectorParameter() {parameterName=EmissiveColorDisabled, defaultValue=<rgba 2 0.1363 0.0382 0>}.RGB
   B=If(
    A=Cosine(None=@n5) {period=2}
    "A > B"=MaterialFunctionCall!(Input=@n6, InputBias=@n7, CurveGain=@n8, CurveBias=@n9) {materialFunction=/Game/Global/MaterialFunctions/MF_PolynomialEaseInOut}.Gain + Bias
    "A < B"=@n6
   )
  )
  EyeAdaptationInverse(LightValueInput=LinearInterpolate(A=Multiply(A=@n1.Linear, B=@n4), B=Multiply(A=@n3, B=@n10), Alpha=PerInstanceCustomData() {constDefaultValue=1, dataIndex=1, desc=Disabled}))
  TextureSample() {texture=/LaserGameplay/Art/Textures/T_Gameplay_LaserEmitter_A_A_D}
  TextureSample() {samplerType=SAMPLERTYPE_Normal, texture=/LaserGameplay/Art/Textures/T_Gameplay_LaserEmitter_A_A_N}
  TextureSample() {samplerType=SAMPLERTYPE_Masks, texture=/LaserGameplay/Art/Textures/T_Gameplay_LaserEmitter_A_A_S}
 ]
}
