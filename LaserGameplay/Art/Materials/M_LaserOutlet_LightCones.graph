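# Unlit, two-sided additive material for laser outlet light cones, drawn as instanced or Niagara meshes.
# UV gradients, optional beam tiling and depth-fade intersection brightening blend dark and light colors into emissive, scaled by time of day and eye-adaptation compensation.
# Opacity combines wobbling gradient falloff, Fresnel and non-mobile view-angle and depth fades.
# Per-instance data drives timestamped in/out transitions and a Disabled state that dims and pulses emissive through an eased sine waveform.
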Material /LaserGameplay/Art/Materials/M_LaserOutlet_LightCones
graph {
 properties {
  bUsedWithInstancedStaticMeshes=true
  bUsedWithNanite=true
  bUsedWithNiagaraMeshParticles=true
  blendMode=BLEND_Additive
  pixelDepthOffsetMode=PDOM_Legacy
  shadingModel=MSM_Unlit
  twoSided=true
  usedShadingModels=Unlit
 }
 nodes [
  n1 = ScalarParameter() {defaultValue=4, parameterName=BrightFade}
  n2 = Multiply(A=OneMinus(None=DepthFade(Opacity=@n1)), B=ComponentMask(None=TextureCoordinate()) {g=true})
  n3 = TextureCoordinate()
  n4 = ComponentMask(None=@n3) {g=true}
  n5 = Power(Base=Add(A=Multiply(A=@n2) {constB=2}, B=@n4)) {constExponent=4}
  n6 = VectorParameter() {parameterName=LightColor, defaultValue=<rgba 4 0 0 1>}
  n7 = Fmod(A=Multiply(A=ComponentMask(None=@n3) {r=true}, B=ScalarParameter() {defaultValue=5, parameterName=BeamTile}), B=Constant() {r=1})
  n8 = Constant() {r=0.5}
  n9 = LinearInterpolate(A=@n4, B=Multiply(A=@n4, B=SphereMask(A=@n7, B=@n8, Radius=@n8) {hardnessPercent=0}), Alpha=ScalarParameter() {parameterName=UseBeamTile})
  n10 = LinearInterpolate(
   A=Multiply(A=@n5, B=VectorParameter() {parameterName=DarkColor, defaultValue=<rgba 0.026 0.0001 0 1>}.RGB)
   B=Add(A=Multiply(A=@n6.RGB, B=@n5), B=@n6.RGB)
   Alpha=@n9
  )
  n11 = LinearInterpolate(A=Multiply(A=@n10) {constB=0}, B=@n10, Alpha=NamedRerouteUsage())
  n12 = Constant() {r=1}
  n13 = Constant() {r=0.75}
  n14 = MaterialFunctionCall(Value_Morning=@n12, Value_Day=@n12, Value_Evening=@n13, Value_Night=@n13) {materialFunction=&Scalar_Time_of_Day}
  n15 = LinearInterpolate(A=@n14.Result, B=Multiply(A=@n14.Result) {constB=0.5}, Alpha=PerInstanceCustomData!(DefaultValue) {dataIndex=1, desc=Disabled}) {constA=1}
  EyeAdaptationInverse(LightValueInput=Multiply(A=Multiply(A=@n11, B=@n15), B=ScalarParameter() {defaultValue=1, parameterName=Emissive Multi}))
  n16 = PerInstanceCustomData() {dataIndex=2, desc=Timestamp}
  n17 = PerInstanceCustomData() {dataIndex=3, desc=TimestampTransition}
  n18 = Multiply(A=Time(), B=LinearInterpolate(Alpha=PerInstanceRandom()) {constA=0.35, constB=0.5})
  n19 = Sine(None=@n18) {period=0.5}
  n20 = Cosine(None=@n18)
  n21 = Subtract(
   A=ScalarParameter() {defaultValue=2, parameterName=OpacityPower}
   B=DotProduct(A=Add(A=@n19, B=@n20), B=ScalarParameter() {defaultValue=0.25, parameterName=GradientSizeWobble})
  )
  n22 = ScalarParameter() {defaultValue=80, parameterName=DepthTipDist}
  n23 = ScalarParameter() {parameterName=DepthBaseDist}
  n24 = ComponentMask(None=TextureCoordinate()) {g=true}
  n25 = Multiply(
   A=MaterialFunctionCall("FadeSensitivity (optional)"=ScalarParameter() {defaultValue=1, parameterName=AngleFade}) {materialFunction=&FadeBasedOnViewAngle}.Result
   B=DepthFade(FadeDistance=LinearInterpolate(A=@n22, B=@n23, Alpha=@n24))
  )
  n26 = Constant() {r=1}
  n27 = MaterialFunctionCall(
   Brightness=Multiply(A=Power(Base=@n9, Exp=@n21), B=ShadingPathSwitch(Default=@n25, Mobile=@n26) {desc=Migrated from FeatureLevelSwitch})
   Exponent=ScalarParameter() {defaultValue=1.25, parameterName=Fresnel Exponent}
  ) {materialFunction=&MF_Fresnel_FX}
  n28 = PerInstanceCustomData() {constDefaultValue=1}
  LinearInterpolate(
   A=Multiply(
    A=MaterialFunctionCall(TimestampA=@n16, TransitionLength_In=@n17) {materialFunction=&MF_Timestamp_InOut_01}.Linear
    B=Multiply(A=@n27.Fresnel, B=@n28)
   )
   B=@n27.Fresnel
   Alpha=PerInstanceCustomData() {constDefaultValue=1, dataIndex=1, desc=Disabled}
  )
  n29 = Time()
  n30 = Cosine(None=@n29) {period=2}
  n31 = ConstantBiasScale(None=Sine(None=@n29) {period=2})
  n32 = MaterialFunctionCall!(Input=@n31, InputBias=Constant() {r=0.0427}, CurveGain=Constant() {r=2.3965}, CurveBias=Constant() {r=0.8553}) {materialFunction=/Game/Global/MaterialFunctions/MF_PolynomialEaseInOut}
  NamedRerouteDeclaration(None=LinearInterpolate(B=If(A=@n30, "A > B"=@n32.Gain + Bias, "A < B"=@n31), Alpha=PerInstanceCustomData() {constDefaultValue=1, dataIndex=1, desc=Disabled}) {constA=1}) {name=Waveform, nodeColor=<rgba 1 0 0.75 1>}
 ]
}
