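# Translucent unlit camera-facing flare sprite for laser outlets, driven by per-instance custom data.
# World position offset billboards and camera-offsets the mesh, blending its size between a randomized fast wobble and a slower eased waveform.
# Emissive tints a flare mask with LightColor, scales it by time of day and eye adaptation, and dims it when Disabled.
# Depth-faded opacity follows timestamped transitions and custom-data intensity, switching to the waveform pulse when Disabled.
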Material /LaserGameplay/Art/Materials/M_LaserOutlet_Mesh_Sprite
graph {
 properties {
  bUsedWithInstancedStaticMeshes=true
  bUsedWithNanite=true
  bUsedWithNiagaraMeshParticles=true
  bUsedWithSplineMeshes=true
  blendMode=BLEND_Translucent
  pixelDepthOffsetMode=PDOM_Legacy
  shadingModel=MSM_Unlit
  twoSided=true
  usedShadingModels=Unlit
 }
 nodes [
  n1 = Multiply(A=Time(), B=LinearInterpolate(Alpha=PerInstanceRandom()) {constA=0.35, constB=0.5})
  n2 = Add(A=Sine(None=@n1) {period=0.5}, B=Cosine(None=@n1))
  n3 = ScalarParameter() {defaultValue=5, parameterName=GradientSizeWobble}
  n4 = Time()
  n5 = ConstantBiasScale(None=Sine(None=@n4) {period=2})
  n6 = If(
   A=Cosine(None=@n4) {period=2}
   "A > B"=MaterialFunctionCall!(Input=@n5, InputBias=Constant() {r=0.0427}, CurveGain=Constant() {r=2.3965}, CurveBias=Constant() {r=0.8553}) {materialFunction=/Game/Global/MaterialFunctions/MF_PolynomialEaseInOut}.Gain + Bias
   "A < B"=@n5
  )
  n7 = Add(
   A=AppendVector(A=ScalarParameter() {defaultValue=25, parameterName=Sprite Scale X}, B=ScalarParameter() {defaultValue=25, parameterName=Sprite Scale Y})
   B=LinearInterpolate(
    A=DotProduct(A=@n2, B=@n3)
    B=NamedRerouteDeclaration(None=@n6) {name=Waveform, nodeColor=<rgba 1 0.5391 0 1>}
    Alpha=PerInstanceCustomData!(DefaultValue)
   )
  )
  n8 = MaterialFunctionCall() {materialFunction=&ObjectScale}
  Add(
   A=MaterialFunctionCall!(Center Location=WorldPosition().XYZ, 0-1 UVs, XY scale=Multiply(A=@n7, B=@n8.Scale X), OverridePivot, "Normals (Optional)") {materialFunction=&Sprite}.World position offset
   B=MaterialFunctionCall(Offset Amount=ScalarParameter() {defaultValue=-5, parameterName=CameraOffset}, World Position=WorldPosition().XYZ) {materialFunction=/Engine/Functions/Engine_MaterialFunctions02/WorldPositionOffset/CameraOffset}.Clamped Camera Offset
  )
  n9 = VectorParameter() {parameterName=LightColor, defaultValue=<rgba 20 0 0 1>}
  n10 = TextureCoordinate()
  n11 = AppendVector(A=Constant() {r=0.375}, B=Constant() {r=0.375})
  n12 = Constant() {r=0.3125}
  n13 = Constant() {r=0.3125}
  n14 = TextureSampleParameter2D(UVs=Add(A=Multiply(A=@n10, B=@n11), B=AppendVector(A=@n12, B=@n13))) {parameterName=FlareMask, samplerSource=SSM_Clamp_WorldGroupSettings, texture=/Game/Effects/Fort_Effects/Textures/Flares/T_Flare_Spikeball}
  n15 = Constant() {r=1}
  n16 = Constant() {r=0.75}
  n17 = MaterialFunctionCall(Value_Morning=@n15, Value_Day=@n15, Value_Evening=@n16, Value_Night=@n16) {materialFunction=&Scalar_Time_of_Day}
  n18 = Multiply(A=@n17.Result) {constB=0.25}
  n19 = PerInstanceCustomData() {dataIndex=1, desc=Disabled}
  EyeAdaptationInverse(LightValueInput=Multiply(A=Multiply(A=@n9.RGB, B=@n14.R), B=LinearInterpolate(A=@n17.Result, B=@n18, Alpha=@n19) {constA=1}))
  n20 = PerInstanceCustomData() {dataIndex=2, desc=Timestamp}
  n21 = PerInstanceCustomData() {dataIndex=3, desc=TimestampTransition}
  n22 = PerInstanceCustomData()
  n23 = Multiply(A=DepthFade(Opacity=@n14.G, FadeDistance=ScalarParameter() {defaultValue=20, parameterName=FadeDistance})) {constB=0.25}
  LinearInterpolate(
   A=Multiply(
    A=MaterialFunctionCall(TimestampA=@n20, TransitionLength_In=@n21) {materialFunction=&MF_Timestamp_InOut_01}.Linear
    B=Multiply(A=@n23, B=@n22)
   )
   B=Multiply(A=@n23, B=NamedRerouteUsage())
   Alpha=PerInstanceCustomData() {dataIndex=1, desc=Disabled}
  ) {constA=1}
 ]
}
