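# Unlit masked material for the core of a laser, built for instanced static meshes with an eye-adaptation-compensated EmissiveCore color and a normal-based offset (WPO_NormalPush).
# A per-instance timestamp in custom data drives MF_Timestamp_InOut_01 as a transition-in only.
# Its SmoothStep controls radius and hardness of a SphereMask along UV1's vertical axis, published as the LaserAnimation reroute.
# PerInstanceRandom flips that axis, so the reveal direction varies per instance.
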Material /LaserGameplay/Art/Materials/M_Laser_Core
graph {
 properties {
  bUsedWithInstancedStaticMeshes=true
  blendMode=BLEND_Masked
  pixelDepthOffsetMode=PDOM_Legacy
  shadingModel=MSM_Unlit
  usedShadingModels=Unlit
 }
 nodes [
  EyeAdaptationInverse(LightValueInput=VectorParameter() {parameterName=EmissiveCore, defaultValue=<rgba 25 1.0664 0.8247 0>}.RGB)
  Multiply(A=VertexNormalWS(), B=ScalarParameter() {defaultValue=0.5, parameterName=WPO_NormalPush})
  n1 = PerInstanceRandom()
  n2 = MaterialFunctionCall(Float2=TextureCoordinate() {coordinateIndex=1}) {materialFunction=&BreakOutFloat2Components}
  n3 = OneMinus(None=@n2.G)
  n4 = MaterialFunctionCall(
   TimestampA=PerInstanceCustomData!(DefaultValue) {dataIndex=2, desc=Time Stamp}
   TransitionLength_In=ScalarParameter() {defaultValue=0.22, parameterName=TransitionIn}
   Transition In Only?=StaticBool() {value=true}
  ) {materialFunction=&MF_Timestamp_InOut_01}
  NamedRerouteDeclaration(None=SphereMask(
    A=If(A=@n1, "A > B"=@n2.G, "A < B"=@n3) {constB=0.5}
    B=Constant()
    Radius=@n4.SmoothStep
    Hardness=LinearInterpolate(Alpha=@n4.SmoothStep) {constA=0.6}
   ) {attenuationRadius=0.05, hardnessPercent=0.6}) {name=LaserAnimation, nodeColor=<rgba 0.4531 1 0 1>}
  NamedRerouteUsage()
 ]
}
